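# First-person fire sound cue for the Egg Launcher weapon.
# A Random node picks one of three equally weighted fire waves, then a Modulator varies pitch slightly upward for variety.
# It uses first-person projectile attenuation, a first-person ranged-fire sound class and the first-person gunfire concurrency group.
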SoundCue /Game/Sounds/Fort_Weapons/Launchers/Egg_Launcher/Egg_Launcher_Fire_1P_Cue
root=Modulator_1
nodes [
 {id=Modulator_1, children [Random_1], properties {pitchMax=1.15, pitchMin=1}}
 {id=Random_1, children [WavePlayer_7,WavePlayer_8,WavePlayer_9], properties {weights [1,1,1]}}
 {id=WavePlayer_7, properties {soundWaveAssetPtr=&/Egg_Launcher_Fire_01}}
 {id=WavePlayer_8, properties {soundWaveAssetPtr=&/Egg_Launcher_Fire_02}}
 {id=WavePlayer_9, properties {soundWaveAssetPtr=&/Egg_Launcher_Fire_03}}
]
properties {
 attenuationSettings=&Fort_Weapon_Projectile_FirstPerson_Athena_Att
 maxDistance=2097152
 soundClassObject=&Fort_WeaponRanged_Fire_1P_PSM
 volumeMultiplier=1
 attenuationOverrides {distanceAlgorithm=NaturalSound, falloffDistance=5000, attenuationShapeExtents {x=1000}}
 concurrencySet [&Weapon_Gunfire_FirstPerson_Conc]
}
