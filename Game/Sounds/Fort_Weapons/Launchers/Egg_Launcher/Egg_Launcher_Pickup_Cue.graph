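# Sound cue for picking up the Egg Launcher weapon, classed as a loot collect sound.
# A single pickup wave feeds a Modulator that slightly randomizes pitch on each play.
# It plays at reduced volume with shared weapon pickup attenuation and references the PickupMaximum_Oldest concurrency set.
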SoundCue /Game/Sounds/Fort_Weapons/Launchers/Egg_Launcher/Egg_Launcher_Pickup_Cue
root=Modulator_0
nodes [
 {id=Modulator_0, children [WavePlayer_0], properties {pitchMax=1.03, pitchMin=0.96}}
 {id=WavePlayer_0, properties {soundWaveAssetPtr=&/Egg_Launcher_PickUp_01}}
]
properties {
 attenuationSettings=&Fort_Weapon_Pickup_Att
 maxDistance=3500
 soundClassObject=&Fort_Loot_Collect
 volumeMultiplier=0.4615
 concurrencyOverrides {maxCount=1, resolutionRule=PreventNew}
 concurrencySet [&PickupMaximum_Oldest]
}
