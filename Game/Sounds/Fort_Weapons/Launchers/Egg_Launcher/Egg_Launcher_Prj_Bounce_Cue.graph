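# Sound cue for the bounce of an Egg Launcher projectile.
# A random node picks one of three equally weighted Easter launcher bounce waves and feeds a modulator that slightly varies pitch.
# It uses the grenade bounce attenuation and the ranged weapon projectile effects sound class.
# A concurrency override caps simultaneous instances at a few and prevents new plays beyond that limit.
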SoundCue /Game/Sounds/Fort_Weapons/Launchers/Egg_Launcher/Egg_Launcher_Prj_Bounce_Cue
root=Modulator_1
nodes [
 {id=Modulator_1, children [Random_2], properties {pitchMax=1.1, pitchMin=0.9}}
 {id=Random_2, children [WavePlayer_10, WavePlayer_11, WavePlayer_12], properties {weights [1,1,1]}}
 {id=WavePlayer_10, properties {soundWaveAssetPtr=/Game/Sounds/Fort_Weapons/Launchers/Easter/Launcher_Easter_BounceA_01}}
 {id=WavePlayer_11, properties {soundWaveAssetPtr=/Game/Sounds/Fort_Weapons/Launchers/Easter/Launcher_Easter_BounceA_02}}
 {id=WavePlayer_12, properties {soundWaveAssetPtr=/Game/Sounds/Fort_Weapons/Launchers/Easter/Launcher_Easter_BounceA_03}}
]
properties {
 attenuationSettings=/Game/Sounds/Attenuation/Fort_Grenade_Bounce_Att
 bOverrideConcurrency=true
 maxDistance=5500
 soundClassObject=&Fort_WeaponRanged_Projectile_Effects
 volumeMultiplier=0.8
 concurrencyOverrides {maxCount=3, resolutionRule=PreventNew}
}
