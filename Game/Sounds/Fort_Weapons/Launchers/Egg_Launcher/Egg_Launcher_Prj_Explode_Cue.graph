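# Sound cue for the Egg Launcher projectile explosion, layered by close, medium and distant range.
# A mixer blends three branches, each with its own explosion attenuation and sound class override.
# The close layer randomly picks one of three waves with slight variation, while medium and distant reuse Easter launcher waves, the distant one with random pitch and a short delay.
# The cue overrides concurrency to prevent new instances at the limit.
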SoundCue /Game/Sounds/Fort_Weapons/Launchers/Egg_Launcher/Egg_Launcher_Prj_Explode_Cue
root=Mixer_5
nodes [
 {id=Attenuation_4, children [Modulator_13], properties {attenuationSettings=/Game/Sounds/Attenuation/Fort_Explosion_Medium_Athena_Att}}
 {
  id=Attenuation_6
  children [Modulator_1]
  properties {
   attenuationSettings=/Game/Sounds/Attenuation/Fort_Explosion_Close_Athena_Att
   attenuationOverrides {falloffDistance=900, attenuationShapeExtents {x=100}}
  }
 }
 {
  id=Attenuation_8
  children [Delay_0]
  properties {
   attenuationSettings=/Game/Sounds/Attenuation/Fort_Explosion_Distant_Athena_Att
   attenuationOverrides {falloffDistance=3000, attenuationShapeExtents {x=1000}}
  }
 }
 {id=Delay_0, children [Modulator_8], properties {delayMax=0.1}}
 {id=Mixer_5, children [SoundClass_6,SoundClass_2,SoundClass_0], properties {inputVolume [1,0.4,0.6]}}
 {id=Modulator_1, children [Random_2], properties {volumeMax=1.1, volumeMin=0.9}}
 {id=Modulator_13, children [WavePlayer_10], properties {pitchMax=1, pitchMin=1, volumeMax=1, volumeMin=1}}
 {id=Modulator_8, children [WavePlayer_11], properties {pitchMax=1.3, pitchMin=0.75, volumeMax=1, volumeMin=1}}
 {id=Random_2, children [WavePlayer_5,WavePlayer_6,WavePlayer_8], properties {weights [1,1,1]}}
 {
  id=SoundClass_0
  children [Attenuation_8]
  properties {soundClassOverride=&Fort_WeaponRanged_Fire_Distant}
 }
 {
  id=SoundClass_2
  children [Attenuation_4]
  properties {soundClassOverride=&Fort_WeaponRanged_Fire_Distant}
 }
 {
  id=SoundClass_6
  children [Attenuation_6]
  properties {soundClassOverride=&Fort_WeaponRanged_Fire_3P_PSM}
 }
 {id=WavePlayer_10, properties {soundWaveAssetPtr=/Game/Sounds/Fort_Weapons/Launchers/Easter/Launcher_Easter_Explode_Medium_01}}
 {id=WavePlayer_11, properties {soundWaveAssetPtr=/Game/Sounds/Fort_Weapons/Launchers/Easter/Launcher_Easter_Explode_Distant_01}}
 {id=WavePlayer_5, properties {soundWaveAssetPtr=&/Egg_Launcher_Explo_Close_03}}
 {id=WavePlayer_6, properties {soundWaveAssetPtr=&/Egg_Launcher_Explo_Close_01}}
 {id=WavePlayer_8, properties {soundWaveAssetPtr=&/Egg_Launcher_Explo_Close_02}}
]
properties {
 bOverrideConcurrency=true
 maxDistance=27500
 volumeMultiplier=1
 attenuationOverrides {falloffDistance=2400, attenuationShapeExtents {x=600}}
 concurrencyOverrides {resolutionRule=PreventNew}
}
