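# Sound cue for the Egg Launcher weapon's reload foley in Fortnite.
# A single reload wave player feeds a Modulator, which randomizes the pitch on each play for variation.
# It uses the shared weapon reload attenuation asset and the weapon foley sound class.
# A concurrency override allows one instance at a time, resolving conflicts by stopping the farthest sound, then preventing new ones.
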SoundCue /Game/Sounds/Fort_Weapons/Launchers/Egg_Launcher/Egg_Launcher_Reload_Cue
root=Modulator_0
nodes [
 {id=Modulator_0, children [WavePlayer_1], properties {pitchMax=1.15}}
 {id=WavePlayer_1, properties {soundWaveAssetPtr=&/Egg_Launcher_Reload_01}}
]
properties {
 attenuationSettings=&Fort_Weapon_Reload_Att
 bOverrideConcurrency=true
 maxDistance=2250
 soundClassObject=&Fort_WeaponFoley
 volumeMultiplier=0.301
 attenuationOverrides {falloffDistance=5000, attenuationShapeExtents {x=1000}}
 concurrencyOverrides {maxCount=1, resolutionRule=StopFarthestThenPreventNew}
}
